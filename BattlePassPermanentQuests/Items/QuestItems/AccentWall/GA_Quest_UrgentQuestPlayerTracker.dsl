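; Gameplay ability for an urgent quest that registers its owning player as a radius-tracked target.
; On activation it gathers every valid player state's Fort controller, without team filtering, although it stores the avatar's team.
; It passes these controllers, the owning pawn, and configurable radius and center-offset bounds to the game state's Radius Tracker component.
; Ending the ability removes tracking for that pawn.
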(event Ability|EventFailedToActivatePassiveAbility
  (params (ActorInfo :type Gameplay%20Ability%20Actor%20Info%20Structure))
  (|CallParentFunction ActorInfo))

(event Ability|EventActivateAbility
  (|CallParentFunction)
  (CallFunction|GetFortPlayerControllers)
  (bind _returnvalue (RadiusTracker|MakeRadiusTrackingInitializer (Variables|Default|GetMax%20Center%20Offset) (Variables|Default|GetMinimum%20Radius) (Variables|Default|GetMaximum%20Radius) 5.0 0.0 "(R=0.000000,G=0.000000,B=0.000000,A=0.400000)" 0 (Variables|GAVars|Data|GetOwningPlayerPawn) (Variables|Default|GetAllPlayerControllers) "Global"))
  (Fortnite|AddActorTrackingInfo (RadiusTracker|GetGameStateRadiusTrackerComponent) _returnvalue))

(event Ability|EventOnEndAbility
  (params (bWasCancelled :type Boolean))
  (|CallParentFunction bWasCancelled)
  (RadiusTracker|RemoveTrackingforActor (RadiusTracker|GetGameStateRadiusTrackerComponent) (Variables|GAVars|Data|GetOwningPlayerPawn)))

(fn GetFortPlayerControllers
  (Variables|Default|SetInstigatorTeam (Team|GetActorTeam (Ability|GetAvatarActorfromActorInfo)))
  (for _index (range (Game|GetNumPlayerStates))
    (Utilities|IsValid (Game|GetPlayerState _index)
      (:Is%20Valid
        (bind _asfort_player_controller (Utilities|Casting|CastToFortPlayerController (PlayerState|GetPlayerController (Game|GetPlayerState _index))))
        (Utilities|Array|AddUnique (Variables|Default|GetAllPlayerControllers) _asfort_player_controller)))))
